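; Mantis weapon Blueprint that adds kill-triggered healing orbs, air-dash-aware fire mode and custom sprint effects while equipped.
; On equip, it listens for pawn deaths and spawns two healing orbs, a tick apart, where the instigator's victims died.
; Tag-query tasks override fire mode to Single while air dash is possible, otherwise Auto, and a sprint MME class is registered.
; Unequip, local unequip and EndPlay each remove these bindings.
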(event FortWeapon|EventOnEquippedAndAttached
  (bind _returnvalue (Game|GetInstigator))
  (bind _returnvalue_2 (Game|GetInstigator))
  (bind _singlefiremodetagquery (Variables|Default|GetSingleFireModeTagQuery))
  (|CallParentFunction)
  (Utilities|FlowControl|Sequence
    (:then_0
      ;[0.02854 0.15 0.0232 0.5] Spawn healing orbs on kill
      (bind (_asynctaskproxy _payload) (GameplayEventRouter|StartListeningToEvent (GameplayEventRouter|GetGlobalEventRouter) "/Script/FortniteGame.FortVerbMessage_PawnDied")
        (:then
          (Variables|Default|SetTask_WaitForPawnDeath _asynctaskproxy))
        (:OnEventReceived
          (if (== (Utilities|Struct|BreakDeathInfo (.DeathInfo (Utilities|Struct|BreakFortVerbMessagePawnDied _payload))) (Class|FortKismetLibrary|GetPlayerStatefromActor _returnvalue_2))
            (Game|SpawnActorfromClass (Variables|Default|GetHealingOrbClass) (Math|Conversions|ToTransform(Vector) (.DeathLocation (Utilities|Struct|BreakDeathInfo (.DeathInfo (Utilities|Struct|BreakFortVerbMessagePawnDied _payload))))) "Undefined" "MultiplyWithRoot" 0 _returnvalue_2)
            (Utilities|FlowControl|DelayUntilNextTick)
            (Game|SpawnActorfromClass (Variables|Default|GetHealingOrbClass) (Math|Conversions|ToTransform(Vector) (.DeathLocation (Utilities|Struct|BreakDeathInfo (.DeathInfo (Utilities|Struct|BreakFortVerbMessagePawnDied _payload))))) "Undefined" "MultiplyWithRoot" 0 _returnvalue_2)))))
    (:then_1
      ;[0.15 0.1212 0.03646 0.5] Switch fire mode from Automatic to Single while air dash is possible. Prevents air dash from triggering while holding primary
      (Utilities|IsValid _returnvalue
        (:Is%20Valid
          (bind _asynctaskproxy_2 (Ability|Tasks|WaitGameplayTagQueryOnActor _returnvalue _singlefiremodetagquery)
            (:then
              (Variables|Default|SetTask_WaitForSingleFireModeEnableQuery _asynctaskproxy_2)
              (bind _asynctaskproxy_3 (Ability|Tasks|WaitGameplayTagQueryOnActor _returnvalue _singlefiremodetagquery "WhenFalse")
                (:then
                  (Variables|Default|SetTask_WaitForSingleFireModeDissableQuery _asynctaskproxy_3))
                (:Triggered
                  (Variables|FortWeapon|SetFireModeOverride "Auto"))))
            (:Triggered
              (Variables|FortWeapon|SetFireModeOverride "Single"))))))
    (:then_2
      ;[0.15 0.03089 0.1478 0.5] Add sprint mode for custom FX
      (bind _output_get (Variables|Default|SetTacticalSprintComponent (Actor|GetComponentbyClass (Game|GetInstigatorController) "/Script/TacticalSprintRuntime.FortPlayerControllerComponent_TacticalSprint")))
      (Sprint|AddCustomSprintMMEClass _output_get (Variables|Default|GetSprintMMEClassToAdd)))))

(event FortWeapon|EventOnUnEquippedAndDetached
  ;[0.15 0.03353 0.04789 0.5] Cleanup binds on unequipped or end play
  (|CallParentFunction)
  (Sprint|RemoveCustomSprintMMEClass (Variables|Default|GetTacticalSprintComponent) (Variables|Default|GetSprintMMEClassToAdd))
  (bind _task_waitforpawndeath (Variables|Default|GetTask_WaitForPawnDeath)
    (:then
      (GameplayEventRouter|StopListeningtoEvent _task_waitforpawndeath))
    (:else
      _))
  (bind _task_waitforsinglefiremodeenablequery (Variables|Default|GetTask_WaitForSingleFireModeEnableQuery)
    (:then
      (Ability|Async|EndAction _task_waitforsinglefiremodeenablequery))
    (:else
      _))
  (bind _task_waitforsinglefiremodedissablequery (Variables|Default|GetTask_WaitForSingleFireModeDissableQuery)
    (:then
      (Ability|Async|EndAction _task_waitforsinglefiremodedissablequery))
    (:else)))

(event EventEndPlay
  (params (EndPlayReason :type EEndPlayReason%20Enum :default "Destroyed"))
  ;[0.15 0.03353 0.04789 0.5] Cleanup binds on unequipped or end play
  (|CallParentFunction EndPlayReason)
  (Sprint|RemoveCustomSprintMMEClass (Variables|Default|GetTacticalSprintComponent) (Variables|Default|GetSprintMMEClassToAdd))
  (bind _task_waitforpawndeath (Variables|Default|GetTask_WaitForPawnDeath)
    (:then
      (GameplayEventRouter|StopListeningtoEvent _task_waitforpawndeath))
    (:else
      _))
  (bind _task_waitforsinglefiremodeenablequery (Variables|Default|GetTask_WaitForSingleFireModeEnableQuery)
    (:then
      (Ability|Async|EndAction _task_waitforsinglefiremodeenablequery))
    (:else
      _))
  (bind _task_waitforsinglefiremodedissablequery (Variables|Default|GetTask_WaitForSingleFireModeDissableQuery)
    (:then
      (Ability|Async|EndAction _task_waitforsinglefiremodedissablequery))
    (:else)))

(event FortWeapon|EventOnLocalUnEquip
  ;[0.15 0.03353 0.04789 0.5] Cleanup binds on unequipped or end play
  (|CallParentFunction)
  (Sprint|RemoveCustomSprintMMEClass (Variables|Default|GetTacticalSprintComponent) (Variables|Default|GetSprintMMEClassToAdd))
  (bind _task_waitforpawndeath (Variables|Default|GetTask_WaitForPawnDeath)
    (:then
      (GameplayEventRouter|StopListeningtoEvent _task_waitforpawndeath))
    (:else
      _))
  (bind _task_waitforsinglefiremodeenablequery (Variables|Default|GetTask_WaitForSingleFireModeEnableQuery)
    (:then
      (Ability|Async|EndAction _task_waitforsinglefiremodeenablequery))
    (:else
      _))
  (bind _task_waitforsinglefiremodedissablequery (Variables|Default|GetTask_WaitForSingleFireModeDissableQuery)
    (:then
      (Ability|Async|EndAction _task_waitforsinglefiremodedissablequery))
    (:else)))

(fn ConstructionScript
  (|CallParentFunction))
